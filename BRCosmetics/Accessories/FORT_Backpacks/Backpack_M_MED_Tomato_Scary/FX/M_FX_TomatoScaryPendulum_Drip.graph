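# Masked, lit Niagara mesh-particle material for a pendulum drip effect on the Tomato Scary backpack.
# A time-scrolling sine along UV V, scaled by dynamic WobbleStr and the vertex normal, forms a wobble vector.
# V blends from a dynamic pivot offset (X reused for Y) to that wobble.
# A panning cellular texture multiplied by particle alpha plus V forms a scalar mask.
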Material /BRCosmetics/Accessories/FORT_Backpacks/Backpack_M_MED_Tomato_Scary/FX/M_FX_TomatoScaryPendulum_Drip
graph {
 properties {
  bUsedWithNiagaraMeshParticles=true
  blendMode=BLEND_Masked
  pixelDepthOffsetMode=PDOM_Legacy
  previewMesh=/Game/Characters/Player/Male/Medium/Bodies/M_MED_Accumulate/FX/SM_Accumulate_Drip_Pendulum
  usedShadingModels=Default Lit
 }
 nodes [
  Constant3Vector()
  Constant() {r=0.1}
  n1 = DynamicParameter() {paramNames [PivotOffsetX,PivotOffsetY,PivotOffsetZ,WobbleStr]}
  n2 = TextureCoordinate()
  n3 = Sine(None=Subtract(A=ComponentMask(None=@n2) {g=true}, B=Time()))
  n4 = DynamicParameter() {paramNames [PivotOffsetX,PivotOffsetY,PivotOffsetZ,WobbleStr]}
  n5 = ComponentMask(None=TextureCoordinate()) {g=true}
  LinearInterpolate(
   A=AppendVector(A=AppendVector(A=@n1.PivotOffsetX, B=@n1.PivotOffsetX), B=@n1.PivotOffsetZ)
   B=Multiply(A=Multiply(A=@n3, B=@n4.WobbleStr) {constB=3}, B=VertexNormalWS())
   Alpha=@n5
  )
  n6 = TextureCoordinate() {vTiling=0.3}
  Multiply(
   A=TextureSample(UVs=Panner(Coordinate=@n6) {speedY=-0.8381}) {samplerType=SAMPLERTYPE_LinearGrayscale, texture=/BRCosmetics/Effects/Fort_Effects/BluGlow/Violentium/Cellular_basecolor}.G
   B=Add(A=ParticleColor().A, B=@n5)
  )
  Normalize()
 ]
}
