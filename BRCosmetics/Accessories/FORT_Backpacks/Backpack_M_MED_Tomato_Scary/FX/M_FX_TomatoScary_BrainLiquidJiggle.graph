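# Masked, two-sided Niagara mesh-particle material for the jiggling brain-liquid effect on the scary Tomato backpack.
# Radial UVs sample a cellular texture, tiled by Fluid_Tiling and scrolled by the FluidYPos dynamic parameter.
# Height-scaled noise plus Erode sets a radial sphere-mask radius, feeding a DitherTemporalAA cutout.
# Offset texture taps derive a normal on High and Epic quality, while lower levels use a flat normal.
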Material /BRCosmetics/Accessories/FORT_Backpacks/Backpack_M_MED_Tomato_Scary/FX/M_FX_TomatoScary_BrainLiquidJiggle
graph {
 properties {bUsedWithNiagaraMeshParticles=true, blendMode=BLEND_Masked, pixelDepthOffsetMode=PDOM_Legacy, twoSided=true, usedShadingModels=Default Lit}
 nodes [
  Constant()
  n1 = MaterialFunctionCall(Vector or UVs=Add(A=TextureCoordinate(), B=ScalarParameter() {defaultValue=-0.5, parameterName=uvcTR})) {materialFunction=&VectorToRadialValue}
  n2 = ScalarParameter() {defaultValue=0.5, parameterName=Limit}
  n3 = VectorParameter() {parameterName=Fluid_Tiling, defaultValue=<rgba 1 0 0 1>}
  n4 = Add(
   A=Multiply(A=@n1.Radial Coordinates, B=ComponentMask(None=@n3.RGB) {g=true, r=true})
   B=AppendVector(A=Constant(), B=DynamicParameter() {paramNames [FluidYPos,Erode,Param3,Param4]}.FluidYPos)
  )
  n5 = TextureSample(UVs=@n4) {samplerType=SAMPLERTYPE_LinearGrayscale, texture=/BRCosmetics/Effects/Fort_Effects/BluGlow/Violentium/Cellular_basecolor}
  n6 = Add(
   A=Multiply(A=ScalarParameter() {parameterName=HeightInfluence}, B=@n5.G)
   B=DynamicParameter() {paramNames [FluidYPos,Erode,Param3,Param4]}.Erode
  )
  MaterialFunctionCall(Alpha Threshold=Multiply(A=SphereMask(A=@n1.Linear Distance, B=@n2, Radius=@n6) {attenuationRadius=0.05}) {constB=0.5}) {materialFunction=&DitherTemporalAA}
  ParticleColor()
  n7 = Constant3Vector() {constant=<rgba 0 0 1 0>}
  n8 = ScalarParameter() {parameterName=NormalsTap_Offs}
  n9 = Constant()
  n10 = AppendVector(A=@n8, B=@n9)
  n11 = Subtract(
   A=@n5.R
   B=TextureSample(UVs=Add(A=@n4, B=@n10)) {samplerType=SAMPLERTYPE_LinearGrayscale, texture=&Cellular_basecolor}.R
  )
  n12 = AppendVector(A=@n9, B=@n8)
  n13 = Subtract(
   A=@n5.R
   B=TextureSample(UVs=Add(A=@n12, B=@n4)) {samplerType=SAMPLERTYPE_LinearGrayscale, texture=&Cellular_basecolor}.R
  )
  n14 = DeriveNormalZ(InXY=Multiply(A=AppendVector(A=@n11, B=@n13)) {constB=2})
  QualitySwitch(Default=@n7, Low=@n7, High=@n14, Medium=@n7, Epic=@n14)
  ScalarParameter() {parameterName=HeightPos}
 ]
}
