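# Masked, two-sided lit material for Niagara mesh particles, named as the liquid blast mesh effect of the scary Tomato backpack cosmetic.
# SubUV water splash color and normal textures supply the liquid detail.
# Splash channels, a radial sphere mask, particle alpha and a cellular texture combine into a breakup mask.
# Dynamic parameters drive a cellular-broken radial band via BandPos and an edge-weighted negative-Z vector via Pulldown; BandWidth is unused.
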Material /BRCosmetics/Accessories/FORT_Backpacks/Backpack_M_MED_Tomato_Scary/FX/M_FX_TomatoScary_LiquidBlastMesh
graph {
 properties {bUsedWithNiagaraMeshParticles=true, blendMode=BLEND_Masked, pixelDepthOffsetMode=PDOM_Legacy, twoSided=true, usedShadingModels=Default Lit}
 nodes [
  Constant() {r=0.1}
  n1 = TextureCoordinate()
  n2 = Constant() {r=0.5}
  Distance(A=@n1, B=@n2)
  n3 = ParticleSubUVProperties()
  n4 = TextureSample(UVs=@n3.TextureCoordinate0) {samplerType=SAMPLERTYPE_LinearColor, texture=/Game/Effects/Fort_Effects/Textures/Water/T_waterSplash}
  n5 = Add(A=@n4.R, B=@n4.R)
  n6 = TextureSample() {samplerType=SAMPLERTYPE_LinearGrayscale, texture=/BRCosmetics/Effects/Fort_Effects/BluGlow/Violentium/Cellular_basecolor}
  Multiply(
   A=Add(A=SphereMask(A=@n1, B=@n2) {attenuationRadius=0.5, hardnessPercent=50}, B=Add(A=@n5, B=@n4.B))
   B=Multiply(A=ParticleColor().A, B=@n6.G)
  )
  n7 = TextureCoordinate()
  n8 = Constant() {r=0.5}
  n9 = DynamicParameter() {paramNames [BandPos,BandWidth,Pulldown,Param4]}
  Multiply(A=SphereMask(A=Distance(A=@n7, B=@n8), B=@n9.BandPos, Radius=@n9.BandPos) {hardnessPercent=0}, B=@n6.G)
  n10 = TextureCoordinate()
  n11 = Constant() {r=0.5}
  Multiply(
   A=Multiply(A=Constant3Vector() {constant=<rgba 0 0 -1 0>}, B=@n9.Pulldown)
   B=OneMinus(None=SphereMask(A=@n10, B=@n11) {attenuationRadius=0.5, hardnessPercent=0})
  )
  ParticleColor()
  TextureSample(UVs=@n3.TextureCoordinate0) {samplerType=SAMPLERTYPE_Normal, texture=/Game/Effects/Fort_Effects/Textures/Water/T_waterSplash_R_N}
 ]
}
