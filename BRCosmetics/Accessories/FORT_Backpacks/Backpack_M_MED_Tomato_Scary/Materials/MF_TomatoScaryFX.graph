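# Material function for the Tomato Scary backpack that paints animated glowing goo and a face into base color and emissive, blended over the incoming material attributes by an FX mask.
# A kill-count level, eased over time and optionally quantized, shifts a curve-atlas color ramp along vertex-color green.
# Two parallax-panned bubble layers and fresnel shade the goo, overlaid with an offsettable face texture and sun-based fake scattering.
# Several leftover branches, including pre-skinned bubble projection, feed nothing.
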MaterialFunction /BRCosmetics/Accessories/FORT_Backpacks/Backpack_M_MED_Tomato_Scary/Materials/MF_TomatoScaryFX
graph {nodes [
  n1 = Subtract(A=Multiply(A=TextureCoordinate() {coordinateIndex=1}) {constB=2})
  Abs(None=@n1)
  n2 = PreSkinnedPosition()
  n3 = ShaderStageSwitch(PixelShader=VertexInterpolator(VS=@n2).PS, VertexShader=@n2)
  n4 = Multiply(
   A=Divide(A=FunctionInput(Preview=@n3) {bUsePreviewValueAsDefault=true, inputName=PreskinnedPos}, B=ObjectRadius())
   B=ScalarParameter() {defaultValue=2, parameterName=WP_BubbleTiling}
  )
  n5 = BumpOffset(Coordinate=Panner(Coordinate=ComponentMask(None=@n4) {b=true, r=true}) {speedY=-0.1}, Height=Constant()) {heightRatio=-1}
  n6 = BumpOffset(Coordinate=Panner(Coordinate=ComponentMask(None=@n4) {b=true, g=true}) {speedY=-0.1}, Height=Constant()) {heightRatio=-1}
  n7 = PreSkinnedNormal()
  n8 = ShaderStageSwitch(PixelShader=VertexInterpolator(VS=@n7).PS, VertexShader=@n7)
  n9 = Abs(None=ChannelMaskParameter(None=FunctionInput(Preview=@n8) {bUsePreviewValueAsDefault=true, inputName=PreskinnedNormal}) {parameterName=PS_NormalChannel}.R)
  ChannelMaskParameter(None=LinearInterpolate(
    A=TextureSample(UVs=@n5) {samplerType=SAMPLERTYPE_Masks, texture=/Game/Characters/Player/Male/Large/Bodies/M_LRG_Slurp_Monster/Textures/T_SlurpBubble}.RGB
    B=TextureSample(UVs=@n6) {samplerType=SAMPLERTYPE_Masks, texture=&T_SlurpBubble}.RGB
    Alpha=Round(None=@n9)
   ))
  n10 = VertexColor()
  n11 = Time()
  Clamp(None=SphereMask(A=@n10.G, B=Constant() {r=1}, Radius=Frac(None=@n11), Hardness=Constant()))
  Constant3Vector() {constant=<rgba 1 0.2684 0 1>}
  n12 = CollectionParameter() {
   collection=&FortniteMaterialParameters
   parameterName=SunlightDirectionalVector
  }
  DotProduct(
   A=MaterialFunctionCall() {materialFunction=&CameraDirectionVector}.Result
   B=ComponentMask(None=@n12) {b=true, g=true, r=true}
  )
  n13 = ScalarParameter() {group=SuperLevelling_Mask, parameterName=TS_KillCount, sortPriority=3}
  n14 = Subtract(A=Time(), B=ScalarParameter() {defaultValue=-7.766279436516721e+18, parameterName=TS_GameTimeFromBP})
  n15 = ScalarParameter() {defaultValue=1, parameterName=TS_Duration}
  n16 = LinearInterpolate(A=Subtract(A=@n13), B=@n13, Alpha=Saturate(None=Divide(A=@n14, B=@n15)))
  n17 = ScalarParameter() {group=SuperLevelling_Mask, parameterName=TS_StartAtLevel, sortPriority=3}
  n18 = ScalarParameter() {defaultValue=3, group=SuperLevelling_Mask, parameterName=TS_StopAtLevel, sortPriority=3}
  n19 = Divide(A=Subtract(A=Clamp(None=@n16, Min=@n17, Max=@n18), B=@n17), B=Subtract(A=@n18, B=@n17))
  n20 = Max(A=Ceil(None=ScalarParameter() {defaultValue=1, group=SuperLevelling_Mask, parameterName=SL_QuantizationSteps, sliderMax=100, sliderMin=1}))
  n21 = Divide(A=Floor(None=Multiply(A=@n19, B=@n20)), B=@n20)
  n22 = Saturate(None=Ceil(None=Subtract(A=@n20)))
  n23 = CurveAtlasRowParameter(CurveTime=OneMinus(None=LinearInterpolate(A=@n19, B=@n21, Alpha=@n22))) {
   atlas=&AthenaCurveAtlas01
   curve=/Game/Accessories/FORT_Backpacks/Backpack_M_MED_Tomato_Scary/Materials/TomatoScaryCurve
   defaultValue=304
   parameterName=TS_ColorCurve
  }
  n24 = CurveAtlasRowParameter(CurveTime=Subtract(A=@n10.G, B=@n23.A)) {
   atlas=&AthenaCurveAtlas01
   curve=&TomatoScaryCurve
   defaultValue=304
   parameterName=TS_ColorCurve
  }
  n25 = TextureSample() {
   samplerType=SAMPLERTYPE_Masks
   texture=/BRCosmetics/Accessories/FORT_Backpacks/Backpack_M_MED_Tomato_Scary/Textures/T_M_MED_Tomato_Scary_Backpack_M
  }
  n26 = FunctionInput(Preview=@n25.RGB) {bUsePreviewValueAsDefault=true, inputName=GMap}
  n27 = GetMaterialAttributes(None=FunctionInput() {inputName=Material, inputType=FunctionInput_MaterialAttributes}) {attributeGetTypes [#1]}
  n28 = Fresnel(Normal=Transform(None=@n27.Normal)) {exponent=0.5}
  n29 = Multiply(A=@n10.G, B=Clamp(None=OneMinus(None=@n28)))
  n30 = Multiply(A=@n1, B=ScalarParameter() {defaultValue=1, parameterName=Tiling_01})
  n31 = Constant()
  n32 = TextureSample(UVs=Panner(Coordinate=BumpOffset(Coordinate=@n30, Height=@n31) {heightRatio=1.5}) {speedX=-0.01, speedY=-0.1}) {samplerType=SAMPLERTYPE_Masks, texture=&T_SlurpBubble}
  n33 = Multiply(A=@n1, B=ScalarParameter() {defaultValue=2, parameterName=Tiling_02})
  n34 = Constant()
  n35 = TextureSample(UVs=Panner(Coordinate=BumpOffset(Coordinate=@n33, Height=@n34) {heightRatio=1}) {speedX=0.025, speedY=-0.2}) {samplerType=SAMPLERTYPE_Masks, texture=&T_SlurpBubble}
  n36 = FunctionInput(Preview=TextureSample() {texture=/BRCosmetics/Accessories/FORT_Backpacks/Backpack_M_MED_Tomato_Scary/Textures/T_M_MED_Tomato_Scary_Backpack_FX}.RGB) {bUsePreviewValueAsDefault=true, inputName=FXMask}
  n37 = Multiply(
   A=Min(A=OneMinus(None=@n32.R), B=OneMinus(None=@n35.R))
   B=OneMinus(None=ChannelMaskParameter(None=@n36) {maskChannel=Blue, parameterName=BubbleMask, defaultValue=<rgba 0 0 1 0>}.R)
  )
  n38 = Clamp(None=@n37, Min=ScalarParameter() {defaultValue=0.75, parameterName=BubbleDarkness})
  n39 = LinearInterpolate(
   A=Multiply(A=@n24) {constB=0.1}
   B=@n24
   Alpha=Multiply(A=ChannelMaskParameter(None=@n26) {parameterName=Gmap_Channel}.R, B=Multiply(A=@n29, B=@n38))
  )
  n40 = TextureCoordinate() {coordinateIndex=1}
  n41 = Add(
   A=MaterialFunctionCall(UVs=Multiply(A=@n40) {constB=0.5, desc=1}, Texture Scale=ScalarParameter() {defaultValue=0.5, parameterName=UV_Scale}) {materialFunction=&ScaleUVsByCenter}.UVs
   B=AppendVector(A=ScalarParameter() {defaultValue=0.25, parameterName=UV_X_Offset}, B=ScalarParameter() {defaultValue=0.2571, parameterName=UV_Y_Offset})
  )
  n42 = ScalarParameter() {parameterName=TS_Face_X}
  n43 = ScalarParameter() {parameterName=TS_Face_Y}
  n44 = TextureSample(UVs=Add(A=@n41, B=AppendVector(A=@n42, B=@n43))) {texture=/BRCosmetics/Accessories/FORT_Backpacks/Backpack_M_MED_Tomato_Scary/Textures/T_M_MED_Tomato_Scary_Backpack_faces}
  n45 = Constant() {r=0.25}
  n46 = MaterialFunctionCall(Float2=@n41) {materialFunction=&BreakOutFloat2Components}
  n47 = Multiply(A=@n44.A, B=Multiply(A=SphereMask(A=@n45, B=@n46.R) {attenuationRadius=0.25}, B=SphereMask(A=@n46.G, B=@n45) {attenuationRadius=0.25}))
  n48 = LinearInterpolate(A=@n39, B=@n44.RGB, Alpha=@n47)
  n49 = NamedRerouteUsage()
  n50 = VectorParameter() {parameterName=InScatteringColorMult, defaultValue=<rgba 1 0.2684 0 1>}
  n51 = LinearInterpolate(
   A=@n48
   B=Multiply(A=Multiply(A=@n49, B=@n50.RGB), B=ScalarParameter() {defaultValue=5, parameterName=InScatteringBrightness})
   Alpha=NamedRerouteUsage()
  )
  n52 = LinearInterpolate(A=@n51, B=NamedRerouteUsage(), Alpha=NamedRerouteUsage())
  n53 = Multiply(A=@n52, B=ScalarParameter() {defaultValue=1, parameterName=DiffuseBrightness}) {constB=2}
  n54 = Multiply(A=@n52, B=ScalarParameter() {defaultValue=0.75, parameterName=Emissive_Brightness})
  n55 = ChannelMaskParameter(None=@n36) {maskChannel=Green, parameterName=MaskChannel, defaultValue=<rgba 0 1 0 0>}
  FunctionOutput(None=BlendMaterialAttributes(
    A=@n27.MaterialAttributes
    B=SetMaterialAttributes!(MaterialAttributes=@n27.MaterialAttributes, BaseColor=@n53, EmissiveColor=@n54) {attributeSetTypes [#2, #3]}
    Alpha=@n55.R
   ))
  LinearInterpolate(B=@n48, Alpha=@n55.R)
  LinearInterpolate(B=@n52, Alpha=@n55.R)
  n56 = NamedRerouteUsage()
  n57 = VectorParameter() {parameterName=SS_ScatteringColorMult, defaultValue=<rgba 0.4041 1 0 1>}
  LinearInterpolate(
   A=@n51
   B=Multiply(A=Multiply(A=@n56, B=@n57.RGB), B=ScalarParameter() {defaultValue=15, parameterName=SS_ScatteringBrightness})
   Alpha=NamedRerouteUsage()
  )
  n58 = PixelNormalWS()
  n59 = Clamp(None=Power(Base=Fresnel(Normal=@n58) {exponent=1}))
  Multiply(A=@n48, B=LinearInterpolate(Alpha=@n59) {constA=0.25}) {constB=0.5}
  n60 = Add(A=ChannelMaskParameter(None=TextureCoordinate() {coordinateIndex=1}) {parameterName=UV_X}.R) {constB=-0.25}
  Multiply(A=@n37, B=ConstantBiasScale(None=Sine(None=@n60)))
  Multiply()
  Multiply()
  Multiply(A=@n48, B=LinearInterpolate(Alpha=@n59) {constA=0.25, constB=0.5}) {constB=0.5}
  NamedRerouteDeclaration(None=@n39) {name=BaseGooColor, nodeColor=<rgba 1 0.0469 0 1>}
  NamedRerouteDeclaration(None=@n47) {name=FaceMask, nodeColor=<rgba 1 0.2109 0 1>}
  n61 = ComponentMask(None=SkyAtmosphereLightDirection()) {b=true, g=true, r=true}
  n62 = PixelNormalWS()
  n63 = DotProduct(A=@n61, B=@n62)
  NamedRerouteDeclaration(None=Power(Base=ConstantBiasScale(None=@n63)) {constExponent=4}) {name=SS_Mask, nodeColor=<rgba 1 0.3984 0 1>}
  NamedRerouteDeclaration(None=@n44.RGB) {name=FaceColor, nodeColor=<rgba 1 0 0.3984 1>}
  n64 = DotProduct(A=CameraVectorWS(), B=@n62)
  n65 = CameraVectorWS()
  n66 = Multiply(A=@n64, B=DotProduct(A=@n61, B=Multiply(A=@n65) {constB=-1}))
  NamedRerouteDeclaration(None=Clamp(None=Power(Base=@n66) {constExponent=4})) {name=SS_Inscatter_Mask, nodeColor=<rgba 0 0.4375 1 1>}
  Normalize(VectorInput=@n1)
  OneMinus(None=@n64)
  PixelNormalWS()
  ScalarParameter() {defaultValue=1, parameterName=MipBias_02}
  ScalarParameter() {parameterName=Debug}
  ScalarParameter() {defaultValue=1, parameterName=MipBias_01}
  TextureCoordinate() {coordinateIndex=1}
  TextureSample() {
   samplerType=SAMPLERTYPE_Normal
   texture=/BRCosmetics/Accessories/FORT_Backpacks/Backpack_M_MED_Tomato_Scary/Textures/T_M_MED_Tomato_Scary_Backpack_N
  }
  VectorParameter() {parameterName=PivotOffset, defaultValue=<rgba 0 0 0 1>}
 ]}
