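# Physics asset for the Tomato Scary medium male backpack cosmetic, defining jiggle bodies for its brain, liquid and straw parts.
# Capsules cover the pack and each dyn_ bone, joined by five constraints with tight angular limits.
# Constraints link straw B to straw A to the pack, brain base to body base, and liquid A and B to liquid base.
# No constraint joins body base or liquid base to the pack.
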PhysicsAsset /BRCosmetics/Accessories/FORT_Backpacks/Backpack_M_MED_Tomato_Scary/Meshes/M_MED_Tomato_Scary_Pack_Physics
bodies {
 dyn_body_base [{
   shapeType=Capsule
   length=4.5879
   radius=2.3114
   shapeName=dyn_body_base_capsule
   center=<2.2955 0 0.1472>
   rotation=<rot 0 -89.9999 -93.6702>
  }]
 dyn_body_brain_base [{shapeType=Capsule, length=0, radius=2.8593, shapeName=dyn_body_brain_base_capsule, rotation=<rot 0 -90 0>}]
 dyn_liquid_A [{
   shapeType=Capsule
   length=1.2195
   radius=2.2339
   shapeName=dyn_liquid_A_capsule
   center=<-2.8962 -0.1805 -1.6252>
   rotation=<rot 0 91.0512 -64.2721>
  }]
 dyn_liquid_B [{
   shapeType=Capsule
   length=4.3664
   radius=0.8481
   shapeName=dyn_liquid_B_capsule
   center=<-2.5348 0.2852 -0.1783>
   rotation=<rot 0 -91.7924 -85.0796>
  }]
 dyn_liquid_base [{
   shapeType=Capsule
   length=6.5271
   radius=1.8564
   shapeName=dyn_liquid_base_capsule
   center=<-4.9959 0.7919 0.6158>
   rotation=<rot 13.8019 -89.9999 34.7127>
  }]
 dyn_strawA [{
   shapeType=Capsule
   length=4.3876
   radius=1.2755
   shapeName=dyn_strawA_capsule
   center=<0.1752 -0.0015 0.0139>
   rotation=<rot 0 87.8337 -47.0028>
  }]
 dyn_strawB [{
   shapeType=Capsule
   length=5.1059
   radius=1.5041
   shapeName=dyn_strawB_capsule
   center=<0.3545 -0.1217 3.4078>
   rotation=<rot -0.3436 -90.0001 6.7188>
  }]
 pack [{shapeType=Capsule, length=17.1018, radius=8.4332, center=<-1.0193 -9.0631 -2.2052>, rotation=<rot -2.4973 -90 20.7812>}]
}
constraints [
 {bone1Name=dyn_body_brain_base, bone2Name=dyn_body_base, swing1LimitDegrees=10, swing2LimitDegrees=10, twistLimitDegrees=10}
 {bone1Name=dyn_liquid_A, bone2Name=dyn_liquid_base, swing1LimitDegrees=5, swing2LimitDegrees=10, twistLimitDegrees=5}
 {bone1Name=dyn_liquid_B, bone2Name=dyn_liquid_base, swing1LimitDegrees=5, swing2LimitDegrees=10, twistLimitDegrees=5}
 {bone1Name=dyn_strawB, bone2Name=dyn_strawA, swing1LimitDegrees=10, swing2LimitDegrees=10, twistLimitDegrees=2}
 {bone1Name=dyn_strawA, bone2Name=pack, swing1LimitDegrees=10, swing2LimitDegrees=10, twistLimitDegrees=2}
]
