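# Anim Blueprint for the Tomato Scary backpack, adding procedural jiggle and wobble to its liquid, body, brain and straws.
# AnimGraph offsets the root of the ref pose, then chains RigidBody, spring AnimDynamics, a brain constraint and additive bone transforms.
# EventGraph drives liquid and brain motion with sine/cosine waves and lowers straw physics alpha for frontend reactivity.
# It also computes a skydive-aware Gravity_override via a deprecated function, which the AnimGraph never reads.
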AnimBlueprint /BRCosmetics/Accessories/FORT_Backpacks/Backpack_M_MED_Tomato_Scary/Meshes/M_MED_Tomato_Scary_Pack_AnimBP
target_skeleton=&/M_MED_Tomato_Scary_Pack_Skeleton
graphs [
 AnimGraph {nodes [
   n1 = "Transform (Modify) Bone - Bone: root"(ComponentPose=Local To Component(LocalPose=Local Space Ref Pose()), Alpha=0.4, Translation="0,1,0", Rotation="0,0,-20") {
    rotationMode=BMM_Additive
    rotationSpace=BCS_BoneSpace
    scaleMode=BMM_Additive
    scaleSpace=BCS_BoneSpace
    translationMode=BMM_Additive
    translationSpace=BCS_BoneSpace
    alphaBoolBlend {blendOption=Linear}
    boneToModify {boneName=root}
    rotation {roll=-20}
    translation {y=1}
   }
   n2 = RigidBody(ComponentPose=@n1) {
    bClampLinearTranslationLimitToRefPose=true
    bOverrideWorldGravity=true
    overridePhysicsAsset=&/M_MED_Tomato_Scary_Pack_Physics
    simulationSpace=BaseBoneSpace
    alphaBoolBlend {blendOption=Linear}
    baseBoneRef {boneName=root}
    componentAppliedLinearAccClamp=<3000 3000 3000>
    componentLinearAccScale=<0.334 0.334 0.334>
    componentLinearVelScale=<0.334 0.334 0.334>
    overrideWorldGravity {z=-980}
   }
   n3 = Anim Dynamics - Bone: dyn_liquid_base(ComponentPose=@n2, Alpha=0.8, ExternalForce="0,0,0") {
    angularDampingOverride=0.999
    angularSpringConstant=800
    bAngularSpring=true
    bGravityOverrideInSimSpace=true
    bOverrideAngularDamping=true
    simulationSpace=RootRelative
    alphaBoolBlend {blendOption=Linear}
    boundBone {boneName=dyn_liquid_base}
    componentAppliedLinearAccClamp=<1000 1000 1000>
    componentLinearAccScale=<1.2 1.2 1.2>
    componentLinearVelScale=<2 2 2>
    gravityOverride {z=-980}
    physicsBodyDefinitions [{
      sphereCollisionRadius=0
      boundBone {boneName=dyn_liquid_base}
      boxExtents=<3 10 15>
      constraintSetup {angularTargetAxis=AxisY, angularLimitsMax=<2 6 11>, angularLimitsMin=<-2 -6 -8>, angularTarget=<0 0 0>}
      localJointOffset=<-2 -0.1 -0.01>
     }]
   }
   n4 = Anim Dynamics - Bone: dyn_liquid_A(ComponentPose=@n3, Alpha=0.8, ExternalForce="0,0,0") {
    angularDampingOverride=0.999
    angularSpringConstant=800
    bAngularSpring=true
    bGravityOverrideInSimSpace=true
    bOverrideAngularDamping=true
    gravityScale=0.7
    simulationSpace=RootRelative
    alphaBoolBlend {blendOption=Linear}
    boundBone {boneName=dyn_liquid_A}
    componentAppliedLinearAccClamp=<1000 1000 1000>
    componentLinearAccScale=<1.2 1.2 1.2>
    componentLinearVelScale=<2 2 2>
    gravityOverride {z=-980}
    physicsBodyDefinitions [{
      sphereCollisionRadius=0
      boundBone {boneName=dyn_liquid_A}
      boxExtents=<3 10 15>
      constraintSetup {angularTargetAxis=AxisY, twistAxis=AxisY, angularLimitsMax=<2 12 30>, angularLimitsMin=<-2 -12 -15>, angularTarget=<0 0 0>}
      localJointOffset=<-2 -0.1 -0.01>
     }]
   }
   n5 = Anim Dynamics - Bone: dyn_liquid_B(ComponentPose=@n4, Alpha=0.8, ExternalForce="0,0,0") {
    angularDampingOverride=0.999
    angularSpringConstant=800
    bAngularSpring=true
    bGravityOverrideInSimSpace=true
    bOverrideAngularDamping=true
    gravityScale=0.7
    simulationSpace=RootRelative
    alphaBoolBlend {blendOption=Linear}
    boundBone {boneName=dyn_liquid_B}
    componentAppliedLinearAccClamp=<1000 1000 1000>
    componentLinearAccScale=<1.2 1.2 1.2>
    componentLinearVelScale=<2 2 2>
    gravityOverride {z=-980}
    physicsBodyDefinitions [{
      sphereCollisionRadius=0
      boundBone {boneName=dyn_liquid_B}
      boxExtents=<3 10 15>
      constraintSetup {angularTargetAxis=AxisY, twistAxis=AxisY, angularLimitsMax=<2 12 30>, angularLimitsMin=<-2 -12 -15>, angularTarget=<0 0 0>}
      localJointOffset=<-2 -0.1 -0.01>
     }]
   }
   n6 = Anim Dynamics - Bone: dyn_body_base(ComponentPose=@n5, Alpha=1, ExternalForce="0,0,0") {
    angularDampingOverride=0.999
    angularSpringConstant=1000
    bAngularSpring=true
    bGravityOverrideInSimSpace=true
    bOverrideAngularDamping=true
    gravityScale=0.5
    alphaBoolBlend {blendOption=Linear}
    boundBone {boneName=dyn_body_base}
    componentAppliedLinearAccClamp=<1000 1000 1000>
    componentLinearAccScale=<1 1 1>
    componentLinearVelScale=<1 1 1>
    physicsBodyDefinitions [{
      sphereCollisionRadius=0
      boundBone {boneName=dyn_body_base}
      boxExtents=<3 10 15>
      constraintSetup {angularTargetAxis=AxisY, twistAxis=AxisY, angularLimitsMax=<0 2 5>, angularLimitsMin=<0 -2 -5>, angularTarget=<0 0 0>}
      localJointOffset=<-2 -0.05 0>
     }]
   }
   n7 = Anim Dynamics - Bone: dyn_body_brain_base(ComponentPose=@n6, Alpha=1, ExternalForce="0,0,0") {
    angularDampingOverride=0.999
    angularSpringConstant=1000
    bAngularSpring=true
    bGravityOverrideInSimSpace=true
    bOverrideAngularDamping=true
    gravityScale=0.5
    simulationSpace=BoneRelative
    alphaBoolBlend {blendOption=Linear}
    boundBone {boneName=dyn_body_brain_base}
    componentAppliedLinearAccClamp=<1000 1000 1000>
    componentLinearAccScale=<1 1 1>
    componentLinearVelScale=<1 1 1>
    physicsBodyDefinitions [{
      sphereCollisionRadius=0
      boundBone {boneName=dyn_body_brain_base}
      boxExtents=<3 10 15>
      constraintSetup {angularTargetAxis=AxisY, twistAxis=AxisY, angularLimitsMax=<15 2 10>, angularLimitsMin=<-15 -2 -10>, angularTarget=<0 0 0>}
      localJointOffset=<-5 -0.05 0>
     }]
    relativeSpaceBone {boneName=pack}
   }
   n8 = "Constraint - dyn_Brain (P)"(ComponentPose=@n7, Alpha=0.5, ConstraintWeights_0=1) {
    alphaBoolBlend {blendOption=Linear}
    boneToModify {boneName=dyn_Brain}
    constraintSetup [{targetBone {boneName=dyn_body_brain_base}}]
    constraintWeights [1]
   }
   n9 = Get frontendAlphaControl() -> (frontendAlphaControl:double)
   n10 = Anim Dynamics - Bone: dyn_strawA(ComponentPose=@n8, Alpha=@n9, ExternalForce="0,0,0") {
    angularDampingOverride=0.999
    angularSpringConstant=1000
    bAngularSpring=true
    bGravityOverrideInSimSpace=true
    bOverrideAngularDamping=true
    gravityScale=0.6
    simulationSpace=BoneRelative
    alphaBoolBlend {blendOption=Linear}
    boundBone {boneName=dyn_strawA}
    componentAppliedLinearAccClamp=<2000 2000 2000>
    componentLinearAccScale=<1 1 1>
    componentLinearVelScale=<1 1 1>
    physicsBodyDefinitions [{
      sphereCollisionRadius=0
      boundBone {boneName=dyn_strawA}
      boxExtents=<3 10 15>
      constraintSetup {angularTargetAxis=AxisY, angularLimitsMax=<5 5 5>, angularLimitsMin=<-5 -5 -5>, angularTarget=<0 0 0>}
      localJointOffset=<-4 -0.05 -1>
     }]
   }
   n11 = Anim Dynamics - Bone: dyn_strawB(ComponentPose=@n10, Alpha=@n9, ExternalForce="0,0,0") {
    angularDampingOverride=0.999
    angularSpringConstant=1000
    bAngularSpring=true
    bGravityOverrideInSimSpace=true
    bOverrideAngularDamping=true
    gravityScale=0.6
    simulationSpace=BoneRelative
    alphaBoolBlend {blendOption=Linear}
    boundBone {boneName=dyn_strawB}
    componentAppliedLinearAccClamp=<2000 2000 2000>
    componentLinearAccScale=<1 1 1>
    componentLinearVelScale=<1 1 1>
    physicsBodyDefinitions [{
      sphereCollisionRadius=0
      boundBone {boneName=dyn_strawB}
      boxExtents=<3 10 15>
      constraintSetup {angularTargetAxis=AxisY, angularLimitsMax=<5 10 10>, angularLimitsMin=<-5 -15 -10>, angularTarget=<0 0 0>}
      localJointOffset=<-4 -0.05 1>
     }]
   }
   n12 = Get Liquid_A() -> (Liquid_A:Vector)
   n13 = "Transform (Modify) Bone - Bone: dyn_liquid_base"(ComponentPose=@n11, Alpha=0.5, Translation=@n12, Rotation="0,0,0", Scale="1,1,1") {
    rotationMode=BMM_Additive
    rotationSpace=BCS_BoneSpace
    scaleMode=BMM_Additive
    scaleSpace=BCS_BoneSpace
    translationMode=BMM_Additive
    translationSpace=BCS_BoneSpace
    alphaBoolBlend {blendOption=Linear}
    boneToModify {boneName=dyn_liquid_base}
   }
   n14 = "Transform (Modify) Bone - Bone: dyn_liquid_A"(ComponentPose=@n13, Alpha=0.5, Translation=@n12, Rotation="0,0,0", Scale="1,1,1") {
    rotationMode=BMM_Additive
    rotationSpace=BCS_BoneSpace
    scaleMode=BMM_Additive
    scaleSpace=BCS_BoneSpace
    translationMode=BMM_Additive
    translationSpace=BCS_BoneSpace
    alphaBoolBlend {blendOption=Linear}
    boneToModify {boneName=dyn_liquid_A}
   }
   n15 = "Transform (Modify) Bone - Bone: dyn_liquid_B"(ComponentPose=@n14, Alpha=1, Translation=@n12, Rotation="0,0,0", Scale="1,1,1") {
    rotationMode=BMM_Additive
    rotationSpace=BCS_BoneSpace
    scaleMode=BMM_Additive
    scaleSpace=BCS_BoneSpace
    translationMode=BMM_Additive
    translationSpace=BCS_BoneSpace
    alphaBoolBlend {blendOption=Linear}
    boneToModify {boneName=dyn_liquid_B}
   }
   n16 = "Transform (Modify) Bone - Bone: dyn_brain"(ComponentPose=@n15, Alpha=0.4, Translation=@n12, Rotation="0,0,0", Scale=Get Brain_A_scale() -> (Brain_A_scale:Vector)) {
    rotationMode=BMM_Additive
    rotationSpace=BCS_BoneSpace
    scaleMode=BMM_Additive
    scaleSpace=BCS_BoneSpace
    translationMode=BMM_Additive
    translationSpace=BCS_BoneSpace
    alphaBoolBlend {blendOption=Linear}
    boneToModify {boneName=dyn_brain}
   }
   Output Pose(Result=Component To Local(ComponentPose=@n16)) {layerGroup=None}
  ]}
 EventGraph {nodes [
   Event BlueprintUpdateAnimation()
   > n1 = Sequence!(execute) -> (then_0, then_1, then_2:Exec)
   > n2 = Gravity Override State Function DEPRECATED(CharacterPartAnimInstance=Self-Reference(), GameTimeInSeconds=GetGameTimeInSeconds(), GravityOverrideParams=Get Gravity Override Params())
   > Set Gravity_override(Gravity_override:Vector=@n2) -> (Output_Get:Vector)
   n3 = float * float(A=1, B=1)
   n4 = GetRealTimeSeconds()
   n5 = float * float(A="Sin (Radians)"(A=float * float(A=@n3, B=@n4)), B=0.1)
   n6 = float * float(A=1, B=0.9)
   n7 = float * float(A="Sin (Radians)"(A=float * float(A=@n6, B=@n4)), B=0.8)
   n8 = float * float(A=1, B=2)
   n9 = float * float(A="Sin (Radians)"(A=float * float(A=@n8, B=@n4)), B=-0.1)
   Set Liquid_A(
    execute=@n1.then_1
    Liquid_A:Vector=MakeVector(
     X=float + float(A=@n7, B=-0.3) {is=title K2Node_CommutativeAssociativeBinaryOperator}
     Y=float + float(A=@n5, B=0.1)
     Z=float + float(A=@n9, B=0.7)
    )
   ) -> (Output_Get:Vector)
   n10 = float * float(A=1, B=0.15)
   n11 = GetGameTimeInSeconds()
   n12 = float * float(A="Cos (Radians)"!(A:double=float * float(A=@n10, B=@n11)) -> (ReturnValue:double), B=0.02)
   n13 = float * float(A=1, B=0.5)
   n14 = float * float(A="Cos (Radians)"(A=float * float(A=@n13, B=@n11)), B=0.06)
   n15 = float * float(A=1, B=0.75)
   n16 = float * float(A="Cos (Radians)"(A=float * float(A=@n15, B=@n11)), B=0.02)
   > Set Brain_A_scale(Brain_A_scale:Vector=MakeVector(X=float + float(A=@n14, B=1), Y=float + float(A=@n16, B=1), Z=float + float(A=@n12, B=1))) -> (Output_Get:Vector)
   Set frontendAlphaControl(
    execute=@n1.then_2
    frontendAlphaControl:double=SelectFloat(A=0.25, B=0.8, bPickA=Get bShouldDisplayFrontendReactivity() -> (bShouldDisplayFrontendReactivity:bool))
   ) -> (Output_Get:double)
   TryGetPawnOwner()
  ]}
]
variables [
 Brain_A_scale {type=Vector}
 frontendAlphaControl {type=number}
 Gravity Override Params {
  type=GravityOverrideParamsStruct
  default {
   skydive_DivingDown {gravityOverride=<0 0 0>, windAmplitude=<0 0 0>, windFrequency=<0 0 0>}
   skydive_FloatingDown {gravityOverride=<0 0 300>, windAmplitude=<0 0 0>, windFrequency=<0 0 0>, windOffset=<0 0 0>}
   skydive_GoingUp {gravityOverride=<0 -50 300>, windAmplitude=<0 0 0>, windFrequency=<0 0 0>}
   skydive_Parachuting {gravityOverride=<0 0 -980>, windAmplitude=<0 0 0>, windFrequency=<0 0 0>}
  }
 }
 Gravity_override {type=Vector, default=<0 0 -980>}
 Liquid_A {type=Vector}
]
